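# Physics asset for the face accessory mesh of the female Critter Streak Alt character, covering dynamic hair bones and the upper body.
# Multiple capsules on head and upper spine, plus one on the neck, form body volumes around the hair.
# The two left front hair bones form a chain from the head with tighter swing toward the tip and zero twist limits.
# The top hair bone swings and twists freely on the head.
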PhysicsAsset /BRCosmetics/Characters/Player/Female/Medium/Bodies/F_MED_Critter_Streak_Alt/Meshes/Parts/F_MED_Critter_Streak_Alt_FaceAcc_Physics
bodies {
 dyn_hairTop_1 [{shapeType=Capsule, length=0.505, radius=0.505, shapeName=dyn_hairTop_1_capsule, rotation=<rot 0 -90 0>}]
 dyn_hair_frt_3_l [{shapeType=Capsule, length=4.5, radius=0.75, shapeName=dyn_hair_frt_3_l_capsule, center=<2.0276 0.8039 0.0002>, rotation=<rot 90 0 0>}]
 dyn_hair_frt_4_l [{
   shapeType=Capsule
   length=5.5
   radius=0.75
   shapeName=dyn_hair_frt_4_l_capsule
   center=<2.2299 0.5199 0.0003>
   rotation=<rot 85.3912 180 180>
  }]
 head [
  {shapeType=Capsule, length=8, radius=6.5, center=<2.0653 0 0.4174>, rotation=<rot 90 0 0>}
  {shapeType=Capsule, length=15, radius=15, shapeName=head_capsule, center=<-7.7173 -0.0002 -21.382>, rotation=<rot 75.0633 0 0>}
 ]
 neck_01 [{shapeType=Capsule, length=0, radius=6.2347, shapeName=neck_01_capsule, center=<-0.38 2.5858 0>}]
 spine_05 [
  {shapeType=Capsule, length=20.75, radius=14.25, center=<-2.4369 -5.2307 0>}
  {shapeType=Capsule, length=15, radius=10.75, shapeName=spine_05_capsule, center=<-6.8311 1.3258 0>}
  {
   shapeType=Capsule
   length=8.75
   radius=4.75
   shapeName=spine_05_capsule
   center=<1.3525 5.2961 -8.0197>
   rotation=<rot 0.8288 58.1784 89.4848>
  }
  {shapeType=Capsule, length=6.75, radius=6.25, shapeName=spine_05_capsule, center=<8.787 0 0>}
 ]
}
constraints [
 {bone1Name=head, bone2Name=neck_01, swing1LimitDegrees=45, swing2LimitDegrees=45, twistLimitDegrees=45}
 {
  bone1Name=neck_01
  bone2Name=spine_05
  swing1LimitDegrees=45
  swing1Motion=Free
  swing2LimitDegrees=45
  swing2Motion=Free
  twistLimitDegrees=45
  twistMotion=Free
 }
 {
  bone1Name=dyn_hairTop_1
  bone2Name=head
  swing1LimitDegrees=45
  swing1Motion=Free
  swing2LimitDegrees=45
  swing2Motion=Free
  twistLimitDegrees=45
  twistMotion=Free
 }
 {bone1Name=dyn_hair_frt_3_l, bone2Name=head, swing1LimitDegrees=45, swing2LimitDegrees=45, twistLimitDegrees=0}
 {bone1Name=dyn_hair_frt_4_l, bone2Name=dyn_hair_frt_3_l, swing1LimitDegrees=20, swing2LimitDegrees=20, twistLimitDegrees=0}
]
