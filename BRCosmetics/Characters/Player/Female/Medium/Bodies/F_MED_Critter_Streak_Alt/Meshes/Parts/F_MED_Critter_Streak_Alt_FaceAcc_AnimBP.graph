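# Post-pose AnimBlueprint for the Critter Streak Alt face accessory mesh, adding secondary motion and corrective hair and hood adjustments.
# It copies the attached parent's pose, then runs RigidBody simulation with a dedicated physics asset, overridden gravity and damped component motion.
# Each update stores the head's yaw relative to neck_01 from reference pose as forward/back tilt.
# That value, range-mapped to alpha, drives chained additive bone-space offsets on front-hair and hood bones.
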AnimBlueprint /BRCosmetics/Characters/Player/Female/Medium/Bodies/F_MED_Critter_Streak_Alt/Meshes/Parts/F_MED_Critter_Streak_Alt_FaceAcc_AnimBP
target_skeleton=&/F_MED_Critter_Streak_Alt_FaceAcc_Skeleton
graphs [
 AnimGraph {nodes [
   n1 = RigidBody(ComponentPose=Local To Component(LocalPose=Copy Pose From Mesh() {bUseAttachedParent=true})) {
    bClampLinearTranslationLimitToRefPose=true
    bOverrideWorldGravity=true
    overridePhysicsAsset=&/F_MED_Critter_Streak_Alt_FaceAcc_Physics
    alphaBoolBlend {blendOption=Linear}
    componentAppliedLinearAccClamp=<3000 3000 3000>
    componentLinearAccScale=<0.2 0.2 0.2>
    componentLinearVelScale=<0.2 0.2 0.2>
    overrideWorldGravity {z=-980}
   }
   n2 = Get head_fwdBwd_rot() -> (head_fwdBwd_rot:double)
   n3 = "Transform (Modify) Bone - Bone: dyn_hair_frt_3_l"(ComponentPose=@n1, Alpha=@n2, Translation="0,2,-7", Rotation="45,0,0", Scale="1,1,1") {
    rotationMode=BMM_Additive
    rotationSpace=BCS_BoneSpace
    translationMode=BMM_Additive
    translationSpace=BCS_BoneSpace
    alphaBoolBlend {blendOption=Linear}
    alphaScaleBiasClamp {bMapRange=true, inRange {max=0, min=-25}, outRange {max=0, min=1}}
    boneToModify {boneName=dyn_hair_frt_3_l}
    rotation {pitch=45}
    translation {y=2, z=-7}
   }
   n4 = "Transform (Modify) Bone - Bone: dyn_hood_l"(ComponentPose=@n3, Alpha=@n2, Translation="0,0,-5", Rotation="0,0,0", Scale="1,1,1") {
    rotationMode=BMM_Additive
    rotationSpace=BCS_BoneSpace
    translationMode=BMM_Additive
    translationSpace=BCS_BoneSpace
    alphaBoolBlend {blendOption=Linear}
    alphaScaleBiasClamp {bMapRange=true, inRange {max=0, min=-25}, outRange {max=0, min=1}}
    boneToModify {boneName=dyn_hood_l}
    translation {z=-5}
   }
   n5 = "Transform (Modify) Bone - Bone: neck_hood_01"(ComponentPose=@n4, Alpha=@n2, Translation="2.5,-1.5,2", Rotation="0,-45,0", Scale="1,1,1") {
    rotationMode=BMM_Additive
    rotationSpace=BCS_BoneSpace
    translationMode=BMM_Additive
    translationSpace=BCS_BoneSpace
    alphaBoolBlend {blendOption=Linear}
    alphaScaleBiasClamp {bMapRange=true, inRange {max=5, min=-25}, outRange {max=0, min=1}}
    boneToModify {boneName=neck_hood_01}
    rotation {yaw=-45}
    translation=<2.5 -1.5 2>
   }
   n6 = "Transform (Modify) Bone - Bone: dyn_hood_l"(ComponentPose=@n5, Alpha=@n2, Translation="-1,-2,0", Rotation="0,0,0", Scale="1,1,1") {
    rotationMode=BMM_Additive
    rotationSpace=BCS_BoneSpace
    translationMode=BMM_Additive
    translationSpace=BCS_BoneSpace
    alphaBoolBlend {blendOption=Linear}
    alphaScaleBiasClamp {bMapRange=true, inRange {max=20}}
    boneToModify {boneName=dyn_hood_l}
    translation=<-1 -2>
   }
   Output Pose(Result=Component To Local(ComponentPose=@n6)) {layerGroup=None}
  ]}
 EventGraph {nodes [
   Event BlueprintUpdateAnimation()
   n1 = GetDeltaTransformFromRefPose(self=GetOwnerSkeletalMeshComponent(), BoneName=head, BaseName=neck_01)
   > Set head_fwdBwd_rot(head_fwdBwd_rot:double=BreakRotator(InRot=BreakTransform(InTransform=@n1).Rotation).Yaw) -> (Output_Get:double)
  ]}
]
variables [head_fwdBwd_rot {type=number}]
